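# Physics asset for the SodaMug medium female body cosmetic, named as the cloth collision set for its skeletal mesh.
# Capsules cover the pelvis, lower spine and feet, spheres cover the thighs, and the calves have no shapes.
# A huge sphere offset far below the root bone puts its top just above the origin, approximating a floor.
# Constraints chain spine and legs to the pelvis with all angular motion free.
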PhysicsAsset /BRCosmetics/Characters/Player/Female/Medium/Bodies/F_MED_SodaMug/Meshes/F_MED_SodaMug_Cloth
bodies {
 calf_l []
 calf_r []
 foot_l [{shapeType=Capsule, length=13.7232, radius=8.6453, shapeName=foot_l_capsule, center=<-3.9254 5.8957 0.0583>, rotation=<rot 0 6.6027 -79.4871>}]
 foot_r [{
   shapeType=Capsule
   length=13.7232
   radius=8.6453
   shapeName=foot_l_capsule
   center=<3.9254 -5.8957 -0.0583>
   rotation=<rot 0 6.6027 100.5129>
  }]
 pelvis [{shapeType=Capsule, length=8.4602, radius=10.8958, shapeName=pelvis_capsule, center=<-3.9816 -1.5687 0>}]
 root [{shapeType=Sphere, radius=300, shapeName=root_sphere, center=<0 0 -289.4513>}]
 spine_01 [{shapeType=Capsule, length=7.4895, radius=9.5718, shapeName=spine_01_capsule, center=<-1.0764 -1.527 0.2019>}]
 spine_02 [{shapeType=Capsule, length=6.3872, radius=8.5667, shapeName=spine_02_capsule, center=<-1.2126 -0.7099 0.484>}]
 thigh_l [{shapeType=Sphere, radius=8.1875, shapeName=thigh_l_sphere, center=<-15.4845 2.2403 0.767>}]
 thigh_r [{shapeType=Sphere, radius=8.1875, shapeName=thigh_l_sphere, center=<15.4845 -2.2403 -0.767>}]
}
constraints [
 {
  bone1Name=spine_01
  bone2Name=pelvis
  swing1LimitDegrees=45
  swing1Motion=Free
  swing2LimitDegrees=45
  swing2Motion=Free
  twistLimitDegrees=45
  twistMotion=Free
 }
 {
  bone1Name=spine_02
  bone2Name=spine_01
  swing1LimitDegrees=45
  swing1Motion=Free
  swing2LimitDegrees=45
  swing2Motion=Free
  twistLimitDegrees=45
  twistMotion=Free
 }
 {
  bone1Name=thigh_l
  bone2Name=pelvis
  swing1LimitDegrees=45
  swing1Motion=Free
  swing2LimitDegrees=45
  swing2Motion=Free
  twistLimitDegrees=45
  twistMotion=Free
 }
 {
  bone1Name=calf_l
  bone2Name=thigh_l
  swing1LimitDegrees=45
  swing1Motion=Free
  swing2LimitDegrees=45
  swing2Motion=Free
  twistLimitDegrees=45
  twistMotion=Free
 }
 {
  bone1Name=foot_l
  bone2Name=calf_l
  swing1LimitDegrees=45
  swing1Motion=Free
  swing2LimitDegrees=45
  swing2Motion=Free
  twistLimitDegrees=45
  twistMotion=Free
 }
 {
  bone1Name=foot_r
  bone2Name=calf_r
  swing1LimitDegrees=45
  swing1Motion=Free
  swing2LimitDegrees=45
  swing2Motion=Free
  twistLimitDegrees=45
  twistMotion=Free
 }
 {
  bone1Name=calf_r
  bone2Name=thigh_r
  swing1LimitDegrees=45
  swing1Motion=Free
  swing2LimitDegrees=45
  swing2Motion=Free
  twistLimitDegrees=45
  twistMotion=Free
 }
 {
  bone1Name=thigh_r
  bone2Name=pelvis
  swing1LimitDegrees=45
  swing1Motion=Free
  swing2LimitDegrees=45
  swing2Motion=Free
  twistLimitDegrees=45
  twistMotion=Free
 }
]
