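# Control Rig for the SodaMug female medium body that runs the shared DeformRig function, then corrects collar and coat bones.
# Deform Collar copies neck swing without twist to neck_no_twist and weights collar offsets by head and upper-arm spherical pose readers and upper-arm attributes.
# Deform Jacket drives coat offsets from thigh pose readers and attributes unless IsCloth is set.
# IsCloth plus IsWearingCape applies extra coat offsets.
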ControlRigBlueprint /BRCosmetics/Characters/Player/Female/Medium/Bodies/F_MED_SodaMug/Meshes/F_MED_SodaMug_ControlRig
graphs [
 New Function_ContainedGraph {nodes [
   RigVMFunctionEntryNode()
   > RigVMFunction_Sequence = RigVMAggregateNode()
   RigUnit_GetTransform = &RigUnit_GetTransform(Item=(Type=Bone,Name="neck_01"), Space=LocalSpace, bInitial=False)
   RigVMFunction_MathQuaternionSwingTwist = &RigVMFunction_MathQuaternionSwingTwist(TwistAxis=(X=1,Y=0,Z=0), Input=@RigUnit_GetTransform.Transform.Rotation)
   RigUnit_SetRotation = &RigUnit_SetRotation(
    Item=(Type=Bone,Name="neck_no_twist")
    Weight=1
    bInitial=False
    Value=@RigVMFunction_MathQuaternionSwingTwist.Swing
    ExecutePin=@RigVMFunction_Sequence.A
   )
   > n1 = &RigUnit_SphericalPoseReader(
    ActiveRegionScaleFactors=(PositiveWidth=1,NegativeWidth=3,PositiveHeight=3,NegativeHeight=3)
    ActiveRegionSize=0.1
    Debug=(bDrawDebug=false,bDraw2D=False,bDrawLocalAxes=False,DebugScale=25,DebugSegments=20,DebugThickness=0.25)
    DriverAxis=(X=1,Y=0,Z=0)
    DriverItem=(Type=Bone,Name="head")
    FalloffRegionScaleFactors=(PositiveWidth=1,NegativeWidth=1,PositiveHeight=1,NegativeHeight=1)
    FalloffSize=0.3
    FlipHeightScaling=False
    FlipWidthScaling=False
    OptionalParentItem=(Type=Bone,Name="spine_05")
    RotationOffset=(X=-90,Y=0,Z=5)
   )
   > &RigUnit_ModifyTransforms(
    ItemToModify="((Item=(Type=Bone,Name=\"dyn_collar_rear_MSTR\"),Transform=(Rotation=(X=0,Y=0,Z=0.2588,W=0.9659),Translation=(X=0,Y=0,Z=0),Scale3D=(X=1,Y=1,Z=1))),(Item=(Type=Bone,Name=\"dyn_collar_rear_01_L\"),Transform=(Rotation=(X=0,Y=0,Z=0.2588,W=0.9659),Translation=(X=0, ...(truncated)"
    WeightMinimum=0
    Weight=@n1.OutputParam
   )
   > n2 = &RigUnit_SphericalPoseReader(
    ActiveRegionScaleFactors=(PositiveWidth=1,NegativeWidth=3,PositiveHeight=3,NegativeHeight=3)
    ActiveRegionSize=0.1
    Debug=(bDrawDebug=false,bDraw2D=False,bDrawLocalAxes=False,DebugScale=25,DebugSegments=20,DebugThickness=0.25)
    DriverAxis=(X=1,Y=0,Z=0)
    DriverItem=(Type=Bone,Name="head")
    FalloffRegionScaleFactors=(PositiveWidth=1,NegativeWidth=1,PositiveHeight=1,NegativeHeight=1)
    FalloffSize=0.3
    FlipHeightScaling=False
    FlipWidthScaling=False
    OptionalParentItem=(Type=Bone,Name="spine_05")
    RotationOffset=(X=-90,Y=0,Z=-15)
   )
   > &RigUnit_ModifyTransforms(
    ItemToModify="((Item=(Type=Bone,Name=\"dyn_collar_rear_MSTR\"),Transform=(Rotation=(X=0,Y=0,Z=0.2588,W=0.9659),Translation=(X=0,Y=5,Z=0),Scale3D=(X=1,Y=1,Z=1))),(Item=(Type=Bone,Name=\"dyn_collar_rear_01_L\"),Transform=(Rotation=(X=0,Y=0,Z=0.1305,W=0.9914),Translation=(X=0, ...(truncated)"
    WeightMinimum=0
    Weight=@n2.OutputParam
   )
   RigVMFunctionReturnNode(ExecuteContext=@RigVMFunction_Sequence.C)
   DISPATCH_RigDispatch_GetAnimAttribute = RigVMDispatchNode(BoneName=root, Default=0, Name=deformRig_l_upperarm_up_alpha) {template="DISPATCH_RigDispatch_GetAnimAttribute(in Name,in BoneName,in Default,out Value,out Found)"}
   &RigUnit_ModifyTransforms(
    ItemToModify="((Item=(Type=Bone,Name=\"dyn_collar_fr_MSTR_L\"),Transform=(Rotation=(X=0,Y=0.1305,Z=0,W=0.9914),Translation=(X=0,Y=0,Z=0),Scale3D=(X=1,Y=1,Z=1))),(Item=(Type=Bone,Name=\"dyn_collar_rear_01_L\"),Transform=(Rotation=(X=0,Y=0,Z=0,W=1),Translation=(X=0,Y=0,Z=0),S ...(truncated)"
    WeightMinimum=0
    Weight=@DISPATCH_RigDispatch_GetAnimAttribute.Value
    ExecutePin=@RigVMFunction_Sequence.B
   )
   > &RigUnit_ModifyTransforms(
    ItemToModify="((Item=(Type=Bone,Name=\"dyn_collar_fr_MSTR_R\"),Transform=(Rotation=(X=0,Y=0.1736,Z=0,W=0.9848),Translation=(X=0,Y=0,Z=0),Scale3D=(X=1,Y=1,Z=1))),(Item=(Type=Bone,Name=\"dyn_collar_rear_01_R\"),Transform=(Rotation=(X=0,Y=0.2588,Z=0,W=0.9659),Translation=(X=0, ...(truncated)"
    WeightMinimum=0
    Weight=RigVMDispatchNode(BoneName=root, Default=0, Name=deformRig_r_upperarm_up_alpha) {template="DISPATCH_RigDispatch_GetAnimAttribute(in Name,in BoneName,in Default,out Value,out Found)"}.Value
   )
   > n3 = &RigUnit_SphericalPoseReader(
    ActiveRegionScaleFactors=(PositiveWidth=0.5,NegativeWidth=3,PositiveHeight=0,NegativeHeight=3)
    ActiveRegionSize=0.1
    Debug=(bDrawDebug=false,DebugScale=25,DebugSegments=20,DebugThickness=0.25)
    DriverAxis=(X=1,Y=0,Z=0)
    DriverItem=(Type=Bone,Name="upperarm_l")
    FalloffRegionScaleFactors=(PositiveWidth=1,NegativeWidth=1,PositiveHeight=1,NegativeHeight=1)
    FalloffSize=0.3
    OptionalParentItem=(Type=Bone)
    RotationOffset=(X=90,Y=0,Z=0)
   )
   > &RigUnit_ModifyTransforms(
    ItemToModify=((Item=(Type=Bone,Name="dyn_collar_fr_MSTR_L"),Transform=(Rotation=(X=-0.2256,Y=0.1925,Z=0.0946,W=0.9503),Translation=(X=-0.435,Y=-0.4574,Z=-1.101),Scale3D=(X=1,Y=1,Z=1))))
    WeightMinimum=0
    Weight=@n3.OutputParam
   )
   > n4 = &RigUnit_SphericalPoseReader(
    ActiveRegionScaleFactors=(PositiveWidth=0.5,NegativeWidth=3,PositiveHeight=0,NegativeHeight=3)
    ActiveRegionSize=0.1
    Debug=(bDrawDebug=false,DebugScale=25,DebugSegments=20,DebugThickness=0.25)
    DriverAxis=(X=-1,Y=0,Z=0)
    DriverItem=(Type=Bone,Name="upperarm_r")
    FalloffRegionScaleFactors=(PositiveWidth=1,NegativeWidth=1,PositiveHeight=1,NegativeHeight=1)
    FalloffSize=0.3
    FlipWidthScaling=true
    OptionalParentItem=(Type=Bone)
    RotationOffset=(X=90,Y=180,Z=180)
   )
   > &RigUnit_ModifyTransforms(
    ItemToModify="((Item=(Type=Bone,Name=\"dyn_collar_fr_MSTR_R\"),Transform=(Rotation=(X=-0.3565,Y=0.2894,Z=-0.1058,W=0.882),Translation=(X=-0.6904,Y=-0.5369,Z=-0.7521),Scale3D=(X=1,Y=1,Z=1))),(Item=(Type=Bone,Name=\"dyn_collar_rear_01_R\"),Transform=(Rotation=(X=0,Y=0,Z=0,W=1 ...(truncated)"
    WeightMinimum=0
    Weight=@n4.OutputParam
   )
  ]}
 New Function_ContainedGraph {nodes [
   n1 = RigVMFunctionEntryNode()
   n2 = &RigVMFunction_Sequence(ExecuteContext=@n1.ExecuteContext)
   RigVMFunction_Sequence = &RigVMFunction_Sequence(ExecuteContext=@n2.A)
   n3 = &RigUnit_SphericalPoseReader(
    ActiveRegionScaleFactors=(PositiveWidth=2,NegativeWidth=1.5,PositiveHeight=1.75,NegativeHeight=1)
    ActiveRegionSize=0.1
    Debug=(bDrawDebug=false,bDraw2D=False,bDrawLocalAxes=False,DebugScale=25,DebugSegments=20,DebugThickness=0.25)
    DriverAxis=(X=-1,Y=0,Z=0)
    DriverItem=(Type=Bone,Name="thigh_l")
    FalloffRegionScaleFactors=(PositiveWidth=1,NegativeWidth=1.2,PositiveHeight=1,NegativeHeight=0.5)
    FalloffSize=0.3
    FlipHeightScaling=False
    FlipWidthScaling=False
    OptionalParentItem=(Type=Bone,Name="pelvis")
    RotationOffset=(X=130,Y=0,Z=-15)
    ExecutePin=@RigVMFunction_Sequence.A
   )
   > RigVMVariableNode(Value=@n3.OutputParam) {variable=spr_l_thigh_up_45}
   > n4 = &RigUnit_SphericalPoseReader(
    ActiveRegionScaleFactors=(PositiveWidth=3,NegativeWidth=0.8,PositiveHeight=2,NegativeHeight=2)
    ActiveRegionSize=0.1
    Debug=(bDrawDebug=false,bDraw2D=False,bDrawLocalAxes=False,DebugScale=25,DebugSegments=20,DebugThickness=0.25)
    DriverAxis=(X=-1,Y=0,Z=0)
    DriverItem=(Type=Bone,Name="thigh_l")
    FalloffRegionScaleFactors=(PositiveWidth=2,NegativeWidth=1.35,PositiveHeight=2,NegativeHeight=2)
    FalloffSize=0.26
    FlipHeightScaling=false
    FlipWidthScaling=false
    OptionalParentItem=(Type=Bone,Name="pelvis")
    RotationOffset=(X=180,Y=-32,Z=0)
   )
   > RigVMVariableNode(Value=@n4.OutputParam) {variable=spr_l_thigh_up_side}
   > n5 = &RigUnit_SphericalPoseReader(
    ActiveRegionScaleFactors=(PositiveWidth=3,NegativeWidth=2,PositiveHeight=2,NegativeHeight=2)
    ActiveRegionSize=0.1
    Debug=(bDrawDebug=true,bDraw2D=False,bDrawLocalAxes=False,DebugScale=25,DebugSegments=20,DebugThickness=0.25)
    DriverAxis=(X=1,Y=0,Z=0)
    DriverItem=(Type=Bone,Name="thigh_r")
    FalloffRegionScaleFactors=(PositiveWidth=2,NegativeWidth=2,PositiveHeight=2,NegativeHeight=2)
    FalloffSize=0.295
    FlipHeightScaling=False
    FlipWidthScaling=true
    OptionalParentItem=(Type=Bone,Name="pelvis")
    RotationOffset=(X=0,Y=-20,Z=0)
   )
   > RigVMVariableNode(Value=@n5.OutputParam) {variable=spr_r_thigh_up_side}
   > n6 = &RigUnit_SphericalPoseReader(
    ActiveRegionScaleFactors=(PositiveWidth=4,NegativeWidth=3,PositiveHeight=2.5,NegativeHeight=2)
    ActiveRegionSize=0.1
    Debug=(bDrawDebug=true,bDraw2D=False,bDrawLocalAxes=False,DebugScale=25,DebugSegments=20,DebugThickness=0.25)
    DriverAxis=(X=1,Y=0,Z=0)
    DriverItem=(Type=Bone,Name="thigh_r")
    FalloffRegionScaleFactors=(PositiveWidth=1,NegativeWidth=4,PositiveHeight=1,NegativeHeight=0.5)
    FalloffSize=0.36
    FlipHeightScaling=False
    FlipWidthScaling=true
    OptionalParentItem=(Type=Bone,Name="pelvis")
    RotationOffset=(X=-35,Y=0,Z=-35)
   )
   > RigVMVariableNode(Value=@n6.OutputParam) {variable=spr_r_thigh_up_45}
   RigVMFunction_ControlFlowBranch = &RigVMFunction_ControlFlowBranch(ExecuteContext=@RigVMFunction_Sequence.B, Condition=RigVMVariableNode() {variable=IsCloth}.Value)
   &RigUnit_ModifyTransforms(
    ItemToModify="((Item=(Type=Bone,Name=\"dyn_coat_fr_MSTR_L\"),Transform=(Rotation=(X=0.2754,Y=-0.0523,Z=-0.0841,W=0.9562),Translation=(X=0,Y=0,Z=0),Scale3D=(X=1,Y=1,Z=1))),(Item=(Type=Bone,Name=\"dyn_coat_MSTR_L\"),Transform=(Rotation=(X=0,Y=0,Z=0,W=1),Translation=(X=0,Y=0,Z ...(truncated)"
    WeightMinimum=0
    Weight=RigVMDispatchNode(BoneName=root, Default=0, Name=deformRig_l_thigh_up_alpha) {template="DISPATCH_RigDispatch_GetAnimAttribute(in Name,in BoneName,in Default,out Value,out Found)"}.Value
    ExecutePin=@RigVMFunction_ControlFlowBranch.False
   )
   > &RigUnit_ModifyTransforms(
    ItemToModify="((Item=(Type=Bone,Name=\"dyn_coat_fr_MSTR_R\"),Transform=(Rotation=(X=-0.0826,Y=-0.1963,Z=-0.0098,W=0.977),Translation=(X=0,Y=0,Z=0),Scale3D=(X=1,Y=1,Z=1))),(Item=(Type=Bone,Name=\"dyn_coat_MSTR_R\"),Transform=(Rotation=(X=0,Y=-0.1102,Z=0,W=0.9939),Translation ...(truncated)"
    WeightMinimum=0
    Weight=RigVMDispatchNode(BoneName=root, Default=0, Name=deformRig_r_thigh_up_alpha) {template="DISPATCH_RigDispatch_GetAnimAttribute(in Name,in BoneName,in Default,out Value,out Found)"}.Value
   )
   > &RigUnit_ModifyTransforms(
    ItemToModify="((Item=(Type=Bone,Name=\"dyn_coat_fr_MSTR_L\"),Transform=(Rotation=(X=0.0706,Y=0.0452,Z=-0.1684,W=0.9821),Translation=(X=-0.0233,Y=-0.0645,Z=0.5509),Scale3D=(X=1,Y=1,Z=1))),(Item=(Type=Bone,Name=\"dyn_coat_MSTR_L\"),Transform=(Rotation=(X=0,Y=0,Z=-0.1032,W=0.9 ...(truncated)"
    WeightMinimum=0
    Weight=RigVMVariableNode() {variable=spr_l_thigh_up_45}.Value
   )
   > &RigUnit_ModifyTransforms(
    ItemToModify="((Item=(Type=Bone,Name=\"dyn_coat_fr_MSTR_L\"),Transform=(Rotation=(X=0.0407,Y=-0.2927,Z=-0.1007,W=0.95),Translation=(X=0,Y=0,Z=0),Scale3D=(X=1,Y=1,Z=1))),(Item=(Type=Bone,Name=\"dyn_coat_MSTR_L\"),Transform=(Rotation=(X=0.0159,Y=-0.3054,Z=-0.0495,W=0.9508),Tr ...(truncated)"
    WeightMinimum=0
    Weight=RigVMVariableNode() {variable=spr_l_thigh_up_side}.Value
   )
   > &RigUnit_ModifyTransforms(
    ItemToModify="((Item=(Type=Bone,Name=\"dyn_coat_fr_MSTR_R\"),Transform=(Rotation=(X=0.0706,Y=0.0452,Z=-0.1684,W=0.9821),Translation=(X=-0.0233,Y=-0.0645,Z=0.5509),Scale3D=(X=1,Y=1,Z=1))),(Item=(Type=Bone,Name=\"dyn_coat_MSTR_R\"),Transform=(Rotation=(X=0,Y=0,Z=-0.1032,W=0.9 ...(truncated)"
    WeightMinimum=0
    Weight=RigVMVariableNode() {variable=spr_r_thigh_up_45}.Value
   )
   > &RigUnit_ModifyTransforms(
    ItemToModify="((Item=(Type=Bone,Name=\"dyn_coat_fr_MSTR_R\"),Transform=(Rotation=(X=0.0407,Y=-0.2927,Z=-0.1007,W=0.95),Translation=(X=0,Y=0,Z=0),Scale3D=(X=1,Y=1,Z=1))),(Item=(Type=Bone,Name=\"dyn_coat_MSTR_R\"),Transform=(Rotation=(X=0.0159,Y=-0.3054,Z=-0.0495,W=0.9508),Tr ...(truncated)"
    WeightMinimum=0
    Weight=RigVMVariableNode() {variable=spr_r_thigh_up_side}.Value
   )
   &RigUnit_ModifyTransforms(
    ItemToModify="((Item=(Type=Bone,Name=\"dyn_coat_fr_MSTR_L\"),Transform=(Rotation=(X=0,Y=0,Z=0,W=1),Translation=(X=0,Y=0,Z=0),Scale3D=(X=1,Y=1,Z=1))),(Item=(Type=Bone,Name=\"dyn_coat_fr_MSTR_R\"),Transform=(Rotation=(X=0,Y=0,Z=0,W=1),Translation=(X=0,Y=0,Z=0),Scale3D=(X=1,Y= ...(truncated)"
    Weight=1
    WeightMinimum=0
    ExecutePin=@RigVMFunction_ControlFlowBranch.True
   )
   RigVMFunctionReturnNode(ExecuteContext=@n2.B)
   DISPATCH_RigDispatch_GetAnimAttribute = RigVMDispatchNode(BoneName=root, Default=0, Name=deformRig_l_thigh_up_alpha) {template="DISPATCH_RigDispatch_GetAnimAttribute(in Name,in BoneName,in Default,out Value,out Found)"}
   &RigUnit_ModifyTransforms(
    ExecutePin=()
    ItemToModify="((Item=(Type=Bone,Name=\"dyn_coat_fr_MSTR_L\"),Transform=(Rotation=(X=0.2754,Y=-0.0523,Z=-0.0841,W=0.9562),Translation=(X=0,Y=0,Z=0),Scale3D=(X=1,Y=1,Z=1))),(Item=(Type=Bone,Name=\"dyn_coat_MSTR_L\"),Transform=(Rotation=(X=0,Y=0,Z=0,W=1),Translation=(X=0,Y=0,Z ...(truncated)"
    WeightMinimum=0
    Weight=@DISPATCH_RigDispatch_GetAnimAttribute.Value
   )
   > &RigUnit_ModifyTransforms(
    ItemToModify=((Item=(Type=Bone,Name="None"),Transform=(Rotation=(X=0,Y=0,Z=0,W=1),Translation=(X=0,Y=0,Z=0),Scale3D=(X=1,Y=1,Z=1))))
    WeightMinimum=0
    Weight=RigVMDispatchNode(BoneName=root, Default=0, Name=deformRig_r_thigh_up_alpha) {template="DISPATCH_RigDispatch_GetAnimAttribute(in Name,in BoneName,in Default,out Value,out Found)"}.Value
   )
   > &RigUnit_ModifyTransforms(
    ItemToModify="((Item=(Type=Bone,Name=\"dyn_coat_fr_MSTR_L\"),Transform=(Rotation=(X=0.0706,Y=0.0452,Z=-0.1684,W=0.9821),Translation=(X=-0.0233,Y=-0.0645,Z=0.5509),Scale3D=(X=1,Y=1,Z=1))),(Item=(Type=Bone,Name=\"dyn_coat_MSTR_L\"),Transform=(Rotation=(X=0,Y=0,Z=-0.1032,W=0.9 ...(truncated)"
    WeightMinimum=0
    Weight=RigVMVariableNode() {variable=spr_l_thigh_up_45}.Value
   )
   > &RigUnit_ModifyTransforms(
    ItemToModify="((Item=(Type=Bone,Name=\"dyn_coat_fr_MSTR_L\"),Transform=(Rotation=(X=0.0407,Y=-0.2927,Z=-0.1007,W=0.95),Translation=(X=0,Y=0,Z=0),Scale3D=(X=1,Y=1,Z=1))),(Item=(Type=Bone,Name=\"dyn_coat_MSTR_L\"),Transform=(Rotation=(X=0.0159,Y=-0.3054,Z=-0.0495,W=0.9508),Tr ...(truncated)"
    WeightMinimum=0
    Weight=RigVMVariableNode() {variable=spr_l_thigh_up_side}.Value
   )
   > &RigUnit_ModifyTransforms(
    ItemToModify="((Item=(Type=Bone,Name=\"dyn_coat_fr_MSTR_R\"),Transform=(Rotation=(X=0.0706,Y=0.0452,Z=-0.1684,W=0.9821),Translation=(X=-0.0233,Y=-0.0645,Z=0.5509),Scale3D=(X=1,Y=1,Z=1))),(Item=(Type=Bone,Name=\"dyn_coat_MSTR_R\"),Transform=(Rotation=(X=0,Y=0,Z=-0.1032,W=0.9 ...(truncated)"
    WeightMinimum=0
    Weight=RigVMVariableNode() {variable=spr_r_thigh_up_45}.Value
   )
   > &RigUnit_ModifyTransforms(
    ItemToModify="((Item=(Type=Bone,Name=\"dyn_coat_fr_MSTR_R\"),Transform=(Rotation=(X=0.0407,Y=-0.2927,Z=-0.1007,W=0.95),Translation=(X=0,Y=0,Z=0),Scale3D=(X=1,Y=1,Z=1))),(Item=(Type=Bone,Name=\"dyn_coat_MSTR_R\"),Transform=(Rotation=(X=0.0159,Y=-0.3054,Z=-0.0495,W=0.9508),Tr ...(truncated)"
    WeightMinimum=0
    Weight=RigVMVariableNode() {variable=spr_r_thigh_up_side}.Value
   )
   &RigUnit_ParentConstraint(
    AdvancedSettings=(RotationOrderForFilter=XZY)
    Child=(Type=Bone,Name="dyn_coat_fr_MSTR_R")
    ExecutePin=()
    Parents=((Item=(Type=Bone,Name="clavicle_r"),Weight=0.016),(Item=(Type=Bone,Name="spine_04"),Weight=0.984))
    Weight=1
   )
   > &RigUnit_ParentConstraint(
    AdvancedSettings=(RotationOrderForFilter=XZY)
    Child=(Type=Bone,Name="dyn_coat_fr_01_R")
    Parents=((Item=(Type=Bone,Name="spine_05"),Weight=0.0111),(Item=(Type=Bone,Name="spine_04"),Weight=0.9889))
    Weight=1
   )
   > &RigUnit_ParentConstraint(
    AdvancedSettings=(RotationOrderForFilter=XZY)
    Child=(Type=Bone,Name="dyn_coat_fr_02_R")
    Parents=((Item=(Type=Bone,Name="spine_01"),Weight=0.095),(Item=(Type=Bone,Name="spine_02"),Weight=0.218),(Item=(Type=Bone,Name="spine_03"),Weight=0.24),(Item=(Type=Bone,Name="spine_04"),Weight=0.447))
    Weight=1
   )
   > &RigUnit_ParentConstraint(
    AdvancedSettings=(RotationOrderForFilter=XZY)
    Child=(Type=Bone,Name="dyn_coat_bk_MSTR")
    Parents="((Item=(Type=Bone,Name=\"spine_01\"),Weight=0.018),(Item=(Type=Bone,Name=\"spine_02\"),Weight=0.046),(Item=(Type=Bone,Name=\"clavicle_l\"),Weight=0.057),(Item=(Type=Bone,Name=\"clavicle_r\"),Weight=0.057),(Item=(Type=Bone,Name=\"spine_03\"),Weight=0.069),(Item=(Type ...(truncated)"
    Weight=1
   )
   > &RigUnit_ParentConstraint(
    AdvancedSettings=(RotationOrderForFilter=XZY)
    Child=(Type=Bone,Name="dyn_coat_bk_01")
    Parents="((Item=(Type=Bone,Name=\"spine_01\"),Weight=0.018),(Item=(Type=Bone,Name=\"spine_02\"),Weight=0.046),(Item=(Type=Bone,Name=\"clavicle_l\"),Weight=0.057),(Item=(Type=Bone,Name=\"clavicle_r\"),Weight=0.057),(Item=(Type=Bone,Name=\"spine_03\"),Weight=0.069),(Item=(Type ...(truncated)"
    Weight=1
   )
   > &RigUnit_ParentConstraint(
    AdvancedSettings=(RotationOrderForFilter=XZY)
    Child=(Type=Bone,Name="dyn_coat_bk_02")
    Parents=((Item=(Type=Bone,Name="spine_02"),Weight=0.114),(Item=(Type=Bone,Name="spine_01"),Weight=0.194),(Item=(Type=Bone,Name="pelvis"),Weight=0.692))
    Weight=1
   )
   > &RigUnit_ParentConstraint(
    AdvancedSettings=(RotationOrderForFilter=XZY)
    Child=(Type=Bone,Name="dyn_coat_mid_MSTR_R")
    Parents="((Item=(Type=Bone,Name=\"spine_03\"),Weight=0.0181),(Item=(Type=Bone,Name=\"upperarm_twist_01_r\"),Weight=0.0574),(Item=(Type=Bone,Name=\"spine_05\"),Weight=0.1883),(Item=(Type=Bone,Name=\"clavicle_r\"),Weight=0.3483),(Item=(Type=Bone,Name=\"spine_04\"),Weight=0.387 ...(truncated)"
    Weight=1
   )
   > &RigUnit_ParentConstraint(
    AdvancedSettings=(RotationOrderForFilter=XZY)
    Child=(Type=Bone,Name="dyn_coat_mid_01_R")
    Parents=((Item=(Type=Bone,Name="spine_02"),Weight=0.0348),(Item=(Type=Bone,Name="spine_03"),Weight=0.0706),(Item=(Type=Bone,Name="spine_05"),Weight=0.1534),(Item=(Type=Bone,Name="clavicle_r"),Weight=0.2444),(Item=(Type=Bone,Name="spine_04"),Weight=0.4969))
    Weight=1
   )
   > &RigUnit_ParentConstraint(
    AdvancedSettings=(RotationOrderForFilter=XZY)
    Child=(Type=Bone,Name="dyn_coat_mid_02_R")
    Parents=((Item=(Type=Bone,Name="spine_02"),Weight=0.248),(Item=(Type=Bone,Name="spine_01"),Weight=0.303),(Item=(Type=Bone,Name="pelvis"),Weight=0.449))
    Weight=1
   )
   > &RigUnit_ParentConstraint(
    AdvancedSettings=(RotationOrderForFilter=XZY)
    Child=(Type=Bone,Name="dyn_coat_MSTR_R")
    Parents=((Item=(Type=Bone,Name="upperarm_twist_01_r"),Weight=0.014),(Item=(Type=Bone,Name="clavicle_r"),Weight=0.0429),(Item=(Type=Bone,Name="spine_05"),Weight=0.1),(Item=(Type=Bone,Name="spine_04"),Weight=0.8431))
    Weight=1
   )
   > &RigUnit_ParentConstraint(
    AdvancedSettings=(RotationOrderForFilter=XZY)
    Child=(Type=Bone,Name="dyn_coat_01_R")
    Parents=((Item=(Type=Bone,Name="spine_04"),Weight=1.0))
    Weight=1
   )
   > &RigUnit_ParentConstraint(
    AdvancedSettings=(RotationOrderForFilter=XZY)
    Child=(Type=Bone,Name="dyn_coat_02_R")
    Parents=((Item=(Type=Bone,Name="spine_04"),Weight=0.017),(Item=(Type=Bone,Name="spine_01"),Weight=0.09),(Item=(Type=Bone,Name="spine_02"),Weight=0.118),(Item=(Type=Bone,Name="pelvis"),Weight=0.775))
    Weight=1
   )
   > &RigUnit_ParentConstraint(
    AdvancedSettings=(RotationOrderForFilter=XZY)
    Child=(Type=Bone,Name="dyn_coat_fr_MSTR_L")
    Parents=((Item=(Type=Bone,Name="clavicle_l"),Weight=0.016),(Item=(Type=Bone,Name="spine_04"),Weight=0.984))
    Weight=1
   )
   > &RigUnit_ParentConstraint(
    AdvancedSettings=(RotationOrderForFilter=XZY)
    Child=(Type=Bone,Name="dyn_coat_fr_01_L")
    Parents=((Item=(Type=Bone,Name="spine_05"),Weight=0.0111),(Item=(Type=Bone,Name="spine_04"),Weight=0.9889))
    Weight=1
   )
   > &RigUnit_ParentConstraint(
    AdvancedSettings=(RotationOrderForFilter=XZY)
    Child=(Type=Bone,Name="dyn_coat_fr_02_L")
    Parents=((Item=(Type=Bone,Name="spine_01"),Weight=0.095),(Item=(Type=Bone,Name="spine_02"),Weight=0.218),(Item=(Type=Bone,Name="spine_03"),Weight=0.24),(Item=(Type=Bone,Name="spine_04"),Weight=0.447))
    Weight=1
   )
   > &RigUnit_ParentConstraint(
    AdvancedSettings=(RotationOrderForFilter=XZY)
    Child=(Type=Bone,Name="dyn_coat_mid_MSTR_L")
    Parents="((Item=(Type=Bone,Name=\"spine_03\"),Weight=0.0181),(Item=(Type=Bone,Name=\"upperarm_twist_01_l\"),Weight=0.0574),(Item=(Type=Bone,Name=\"spine_05\"),Weight=0.1883),(Item=(Type=Bone,Name=\"clavicle_l\"),Weight=0.3483),(Item=(Type=Bone,Name=\"spine_04\"),Weight=0.387 ...(truncated)"
    Weight=1
   )
   > &RigUnit_ParentConstraint(
    AdvancedSettings=(RotationOrderForFilter=XZY)
    Child=(Type=Bone,Name="dyn_coat_mid_01_L")
    Parents=((Item=(Type=Bone,Name="spine_02"),Weight=0.0348),(Item=(Type=Bone,Name="spine_03"),Weight=0.0706),(Item=(Type=Bone,Name="spine_05"),Weight=0.1534),(Item=(Type=Bone,Name="clavicle_l"),Weight=0.2444),(Item=(Type=Bone,Name="spine_04"),Weight=0.4969))
    Weight=1
   )
   > &RigUnit_ParentConstraint(
    AdvancedSettings=(RotationOrderForFilter=XZY)
    Child=(Type=Bone,Name="dyn_coat_mid_02_L")
    Parents=((Item=(Type=Bone,Name="spine_02"),Weight=0.248),(Item=(Type=Bone,Name="spine_01"),Weight=0.303),(Item=(Type=Bone,Name="pelvis"),Weight=0.449))
    Weight=1
   )
   > &RigUnit_ParentConstraint(
    AdvancedSettings=(RotationOrderForFilter=XZY)
    Child=(Type=Bone,Name="dyn_coat_MSTR_L")
    Parents=((Item=(Type=Bone,Name="upperarm_twist_01_l"),Weight=0.014),(Item=(Type=Bone,Name="clavicle_l"),Weight=0.0429),(Item=(Type=Bone,Name="spine_05"),Weight=0.1),(Item=(Type=Bone,Name="spine_04"),Weight=0.8431))
    Weight=1
   )
   > &RigUnit_ParentConstraint(
    AdvancedSettings=(RotationOrderForFilter=XZY)
    Child=(Type=Bone,Name="dyn_coat_01_L")
    Parents=((Item=(Type=Bone,Name="spine_04"),Weight=1.0))
    Weight=1
   )
   > &RigUnit_ParentConstraint(
    AdvancedSettings=(RotationOrderForFilter=XZY)
    Child=(Type=Bone,Name="dyn_coat_02_L")
    Parents=((Item=(Type=Bone,Name="spine_04"),Weight=0.017),(Item=(Type=Bone,Name="spine_01"),Weight=0.09),(Item=(Type=Bone,Name="spine_02"),Weight=0.118),(Item=(Type=Bone,Name="pelvis"),Weight=0.775))
    Weight=1
   )
  ]}
 RigVMFunctionLibrary {nodes [
   Deform Collar = RigVMCollapseNode(ExecuteContext=())
   Deform Jacket = RigVMCollapseNode(ExecuteContext=())
  ]}
 RigVMFunction_Sequence_1_ContainedGraph {nodes [
   n1 = RigVMFunctionEntryNode()
   RigVMFunction_Sequence = &RigVMFunction_Sequence(ExecuteContext=@n1.ExecuteContext)
   n2 = &RigVMFunction_Sequence(ExecuteContext=@RigVMFunction_Sequence.B)
   RigVMFunctionReturnNode(A=@RigVMFunction_Sequence.A, B=@n2.A, C=@n2.B)
  ]}
 RigVMModel {nodes [
   RigUnit_BeginExecution = &RigUnit_BeginExecution()
   FN_DeformRig_FNC = RigVMFunctionReferenceNode(
    DeformRig_ Curve="(deformRig_l_upperarm_up_alpha_72_#1=True,deformRig_r_upperarm_up_alpha_73_#2=True,deformRig_l_upperarm_fr_alpha_74_#3=True,deformRig_r_upperarm_fr_alpha_75_CB51B80C4 ...(truncated)"
    DeformRig_Bool="(upperarm_1_#4=true,elbow_3_#5=true,wrist_7_#6=False,groin_8_#7=True,glute_17_#8=false,knee_9_40C408324CA2 ...(truncated)"
    DeformRig_Debug=(upperarm_debug_18_#4=False,wrist_debug_20_#6=False,lat_debug_26_#9=False,groinGlute_debug_27_#7=False)
    DeformRig_Value="(upperarm_up_18_#10=(Rotation=(X=0.0,Y=-0.0,Z=0.0,W=1.0),Translation=(X=0.0,Y=0.0,Z=3),Scale3D=(X=0.70,Y=1.0,Z=1.0)),upperarm_fwd_fr_19_#11=(Rotation=(X=0.0,Y=-0.0,Z=0.0,W=1.0),Translation=(X=0.0,Y= ...(truncated)"
    ExecuteContext=@RigUnit_BeginExecution.ExecutePin
   )
   > Deform Collar = RigVMFunctionReferenceNode()
   > Deform Jacket = RigVMFunctionReferenceNode()
   RigVMFunction_MathBoolAnd = &RigVMFunction_MathBoolAnd(A=RigVMVariableNode() {variable=IsCloth}.Value, B=RigVMVariableNode() {variable=IsWearingCape}.Value)
   RigVMFunction_ControlFlowBranch = &RigVMFunction_ControlFlowBranch(ExecuteContext=@"Deform Jacket".ExecuteContext, Condition=@RigVMFunction_MathBoolAnd.Result)
   &RigUnit_ModifyTransforms(
    ItemToModify="((Item=(Type=Bone,Name=\"dyn_coat_mid_04_R\"),Transform=(Rotation=(X=0,Y=0,Z=0,W=1),Translation=(X=0,Y=0,Z=2),Scale3D=(X=1,Y=1,Z=1))),(Item=(Type=Bone,Name=\"dyn_coat_mid_04_L\"),Transform=(Rotation=(X=0,Y=0,Z=0,W=1),Translation=(X=0,Y=0,Z=2),Scale3D=(X=1,Y=1, ...(truncated)"
    Weight=1
    WeightMinimum=0
    ExecutePin=@RigVMFunction_ControlFlowBranch.True
   )
  ]}
]
hierarchy [
 root {loc=<0 0 0>, rot=<0 0 0 1>}
 pelvis {parent=root, loc=<0 -0.5301 98.0571>, rot=<0.0127 0.707 -0.0127 -0.707>}
 spine_01 {parent=pelvis, loc=<6.1208 -0.9328 0>, rot=<0 0 0.034 -0.9994>}
 spine_02 {parent=spine_01, loc=<6.5996 -0.2095 0>, rot=<0 0 0.0238 -0.9997>}
 spine_03 {parent=spine_02, loc=<6.4914 -0.1237 0>, rot=<0 0 0.0149 -0.9999>}
 spine_04 {parent=spine_03, loc=<6.9598 0.0794 0>, rot=<0 0 -0.0132 -0.9999>}
 spine_05 {parent=spine_04, loc=<6.9987 0 0>, rot=<0 0 -0.0015 -1>}
 attach_cape {parent=spine_05, loc=<9.3216 -8.4501 0>, rot=<0.0616 -0.7044 -0.0616 -0.7044>}
 neck_01 {parent=spine_05, loc=<11.3138 0.1529 0>, rot=<0 0 -0.2003 -0.9797>}
 neck_02 {parent=neck_01, loc=<4.5627 0 0>, rot=<0 0 0.0504 -0.9987>}
 head {parent=neck_02, loc=<5.3914 0 0>, rot=<0 0 0.115 -0.9934>}
 attach_backpack {parent=spine_05, loc=<0 -8.35 0>, rot=<-0.0093 -0.707 0.0093 -0.707>}
 clavicle_l {parent=spine_05, loc=<5.4831 -0.1053 -3.2946>, rot=<0.0465 -0.7289 -0.009 -0.6829>}
 upperarm_l {parent=clavicle_l, loc=<9.8658 -1.2136 -0.4105>, rot=<0.0164 -0.3245 -0.0269 -0.9454>}
 lowerarm_l {parent=upperarm_l, loc=<26.465 -0.8345 0.259>, rot=<0.0036 0.0161 -0.2298 -0.9731>}
 hand_l {parent=lowerarm_l, loc=<22.7355 -0.3203 -0.6105>, rot=<-0.6784 -0.0223 0.0735 -0.7307>}
 thumb_01_l {parent=hand_l, loc=<1.3118 -0.3191 -2.459>, rot=<0.6546 -0.4042 0.0664 -0.6354>}
 thumb_02_l {parent=thumb_01_l, loc=<5.9336 -0.0448 0.0401>, rot=<-0.0037 0.0213 0.298 -0.9543>}
 thumb_03_l {parent=thumb_02_l, loc=<2.2223 0.3682 -0.0907>, rot=<0.0118 -0.013 0.0066 -0.9998>}
 index_metacarpal_l {parent=hand_l, loc=<1.6535 0.0183 -2.96>, rot=<0.0158 -0.0124 -0.0173 -0.9996>}
 index_01_l {parent=index_metacarpal_l, loc=<8.3008 0.0357 -0.0524>, rot=<-0.0044 0.0485 0.0924 -0.9945>}
 index_02_l {parent=index_01_l, loc=<3.3587 -0.3004 0.0452>, rot=<-0.002 0.0027 0.3225 -0.9466>}
 index_03_l {parent=index_02_l, loc=<2.5578 -0.123 -0.0252>, rot=<-0.002 0.0067 0.2558 -0.9667>}
 middle_metacarpal_l {parent=hand_l, loc=<2.1856 -0.1725 -0.5984>, rot=<0.006 0.0168 -0.0329 -0.9993>}
 middle_01_l {parent=middle_metacarpal_l, loc=<7.3275 -0.2447 0.0129>, rot=<-0.0208 0.0042 0.0858 -0.9961>}
 middle_02_l {parent=middle_01_l, loc=<4.0123 0.1401 0.1175>, rot=<-0.0218 0.0258 0.3512 -0.9357>}
 middle_03_l {parent=middle_02_l, loc=<2.6077 0.0534 0.1373>, rot=<0.0082 0.0188 0.1708 -0.9851>}
 ring_metacarpal_l {parent=hand_l, loc=<2.2074 -0.3625 1.3638>, rot=<-0.0284 0.0722 -0.0458 -0.9959>}
 ring_01_l {parent=ring_metacarpal_l, loc=<7.0295 -0.6197 -0.1446>, rot=<-0.0489 -0.0252 0.1892 -0.9804>}
 ring_02_l {parent=ring_01_l, loc=<3.6174 0.4647 -0.1027>, rot=<0 0 0.2806 -0.9598>}
 ring_03_l {parent=ring_02_l, loc=<2.4152 0.114 0.0199>, rot=<0 0 0.1189 -0.9929>}
 pinky_metacarpal_l {parent=hand_l, loc=<1.7616 -0.9753 3.3814>, rot=<-0.0397 0.0814 -0.069 -0.9935>}
 pinky_01_l {parent=pinky_metacarpal_l, loc=<7.1709 -0.4831 -0.1094>, rot=<-0.0654 -0.0035 0.2242 -0.9723>}
 pinky_02_l {parent=pinky_01_l, loc=<2.8274 0.0893 -0.1043>, rot=<0.0037 0.0011 0.2854 -0.9584>}
 pinky_03_l {parent=pinky_02_l, loc=<2.2046 0.0307 -0.0078>, rot=<-0.0002 0.0098 0.1445 -0.9895>}
 lowerarm_twist_02_l {parent=lowerarm_l, loc=<10.6707 0 0>, rot=<0 0 0 1>}
 lowerarm_twist_01_l {parent=lowerarm_l, loc=<18.316 0 0>, rot=<0 0 0 1>}
 dyn_glove_strap_b_l {parent=lowerarm_twist_01_l, loc=<2.4601 -3.854 1.228>, rot=<0.6683 -0.7379 -0.0744 -0.0575>}
 dyn_glove_strap_a_l {parent=lowerarm_twist_01_l, loc=<-0.8297 -4.0602 1.6166>, rot=<0.6564 -0.742 -0.0854 -0.1066>}
 deform_elbow_in_base_l {parent=lowerarm_l, loc=<0.1834 3.6062 -0.2572>, rot=<0 0 0 -1>}
 deform_elbow_in_l {parent=deform_elbow_in_base_l, loc=<0 0 0>, rot=<0 0 0 1>}
 deform_elbow_out_base_l {parent=lowerarm_l, loc=<0 -4 0>, rot=<0 0 0 1>}
 deform_elbow_out_l {parent=deform_elbow_out_base_l, loc=<0 0 0>, rot=<0 0 0 1>}
 upperarm_twist_01_l {parent=upperarm_l, loc=<7.2135 -0.1134 0>, rot=<0 0 0 1>}
 deform_upperarm_bk_base_l {parent=upperarm_twist_01_l, loc=<-6.4053 -4.3746 0.314>, rot=<0 0 0 -1>}
 deform_upperarm_bk_l {parent=deform_upperarm_bk_base_l, loc=<0 0 0>, rot=<0 0 0 1>}
 deform_upperarm_fr_base_l {parent=upperarm_twist_01_l, loc=<-5.6312 5.092 0.3354>, rot=<0 0 0 -1>}
 deform_upperarm_fr_l {parent=deform_upperarm_fr_base_l, loc=<0 0 0>, rot=<0 0 0 1>}
 deform_upperarm_up_base_l {parent=upperarm_twist_01_l, loc=<-8.891 0.4825 3.9517>, rot=<0 0 0 -1>}
 deform_upperarm_up_l {parent=deform_upperarm_up_base_l, loc=<0 0 0>, rot=<0 0 0 1>}
 dyn_upperarm_tag_l {parent=upperarm_twist_01_l, loc=<7.2296 -4.1514 2.5383>, rot=<-0.4321 0.5068 0.4845 -0.5672>}
 upperarm_twist_02_l {parent=upperarm_l, loc=<19.7998 -0.1134 0>, rot=<0 0 0 1>}
 clavicle_r {parent=spine_05, loc=<5.4831 -0.1053 3.2946>, rot=<0.7289 0.0465 -0.6829 0.009>}
 upperarm_r {parent=clavicle_r, loc=<-9.8658 1.2136 0.4105>, rot=<0.0164 -0.3245 -0.0269 -0.9454>}
 lowerarm_r {parent=upperarm_r, loc=<-26.465 0.8345 -0.259>, rot=<0.0036 0.0161 -0.2298 -0.9731>}
 hand_r {parent=lowerarm_r, loc=<-22.7355 0.3203 0.6106>, rot=<-0.6784 -0.0223 0.0735 -0.7307>}
 thumb_01_r {parent=hand_r, loc=<-1.3118 0.3191 2.459>, rot=<0.6546 -0.4042 0.0664 -0.6354>}
 thumb_02_r {parent=thumb_01_r, loc=<-5.9336 0.0448 -0.0401>, rot=<-0.0037 0.0213 0.298 -0.9543>}
 thumb_03_r {parent=thumb_02_r, loc=<-2.2223 -0.3682 0.0907>, rot=<0.0118 -0.013 0.0066 -0.9998>}
 index_metacarpal_r {parent=hand_r, loc=<-1.6535 -0.0183 2.96>, rot=<0.0158 -0.0124 -0.0173 -0.9996>}
 index_01_r {parent=index_metacarpal_r, loc=<-8.3008 -0.0357 0.0524>, rot=<-0.0044 0.0485 0.0924 -0.9945>}
 index_02_r {parent=index_01_r, loc=<-3.3587 0.3004 -0.0452>, rot=<-0.002 0.0027 0.3225 -0.9466>}
 index_03_r {parent=index_02_r, loc=<-2.5578 0.123 0.0252>, rot=<-0.002 0.0067 0.2558 -0.9667>}
 middle_metacarpal_r {parent=hand_r, loc=<-2.1856 0.1725 0.5984>, rot=<0.006 0.0168 -0.0329 -0.9993>}
 middle_01_r {parent=middle_metacarpal_r, loc=<-7.3275 0.2447 -0.0129>, rot=<-0.0208 0.0042 0.0858 -0.9961>}
 middle_02_r {parent=middle_01_r, loc=<-4.0123 -0.1401 -0.1175>, rot=<-0.0218 0.0258 0.3512 -0.9357>}
 middle_03_r {parent=middle_02_r, loc=<-2.6077 -0.0534 -0.1373>, rot=<0.0082 0.0188 0.1708 -0.9851>}
 ring_metacarpal_r {parent=hand_r, loc=<-2.2074 0.3625 -1.3638>, rot=<-0.0284 0.0722 -0.0458 -0.9959>}
 ring_01_r {parent=ring_metacarpal_r, loc=<-7.0295 0.6197 0.1446>, rot=<-0.0489 -0.0252 0.1892 -0.9804>}
 ring_02_r {parent=ring_01_r, loc=<-3.6174 -0.4647 0.1027>, rot=<0 0 0.2806 -0.9598>}
 ring_03_r {parent=ring_02_r, loc=<-2.4152 -0.114 -0.0199>, rot=<0 0 0.1189 -0.9929>}
 pinky_metacarpal_r {parent=hand_r, loc=<-1.7616 0.9753 -3.3814>, rot=<-0.0397 0.0814 -0.069 -0.9935>}
 pinky_01_r {parent=pinky_metacarpal_r, loc=<-7.1709 0.4831 0.1094>, rot=<-0.0654 -0.0035 0.2242 -0.9723>}
 pinky_02_r {parent=pinky_01_r, loc=<-2.8274 -0.0893 0.1043>, rot=<0.0037 0.0011 0.2854 -0.9584>}
 pinky_03_r {parent=pinky_02_r, loc=<-2.2046 -0.0307 0.0078>, rot=<-0.0002 0.0098 0.1445 -0.9895>}
 lowerarm_twist_02_r {parent=lowerarm_r, loc=<-10.6707 0 0>, rot=<0 0 0 -1>}
 lowerarm_twist_01_r {parent=lowerarm_r, loc=<-18.316 0 0>, rot=<0 0 0 -1>}
 dyn_glove_strap_a_r {parent=lowerarm_twist_01_r, loc=<0.8298 4.0601 -1.6168>, rot=<0.6564 -0.742 -0.0854 -0.1066>}
 dyn_glove_strap_b_r {parent=lowerarm_twist_01_r, loc=<-2.4604 3.8541 -1.2276>, rot=<0.6683 -0.7379 -0.0744 -0.0575>}
 deform_elbow_in_base_r {parent=lowerarm_r, loc=<-0.1834 -3.6062 0.2572>, rot=<0 0 0 -1>}
 deform_elbow_in_r {parent=deform_elbow_in_base_r, loc=<0 0 0>, rot=<0 0 0 1>}
 deform_elbow_out_base_r {parent=lowerarm_r, loc=<0 4 -0.0001>, rot=<0 0 0 -1>}
 deform_elbow_out_r {parent=deform_elbow_out_base_r, loc=<0 0 0>, rot=<0 0 0 1>}
 upperarm_twist_01_r {parent=upperarm_r, loc=<-7.2135 0.1134 0>, rot=<0 0 0 -1>}
 deform_upperarm_bk_base_r {parent=upperarm_twist_01_r, loc=<6.4053 4.3746 -0.314>, rot=<0 0 0 -1>}
 deform_upperarm_bk_r {parent=deform_upperarm_bk_base_r, loc=<0 0 0>, rot=<0 0 0 1>}
 deform_upperarm_fr_base_r {parent=upperarm_twist_01_r, loc=<5.6312 -5.092 -0.3354>, rot=<0 0 0 -1>}
 deform_upperarm_fr_r {parent=deform_upperarm_fr_base_r, loc=<0 0 0>, rot=<0 0 0 1>}
 deform_upperarm_up_base_r {parent=upperarm_twist_01_r, loc=<8.891 -0.4825 -3.9517>, rot=<0 0 0 -1>}
 deform_upperarm_up_r {parent=deform_upperarm_up_base_r, loc=<0 0 0>, rot=<0 0 0 1>}
 dyn_upperarm_tag_r {parent=upperarm_twist_01_r, loc=<-7.2299 4.1514 -2.538>, rot=<-0.4321 0.5068 0.4845 -0.5672>}
 upperarm_twist_02_r {parent=upperarm_r, loc=<-19.7998 0.1134 0>, rot=<0 0 0 -1>}
 dyn_collar_fr_MSTR_L {parent=spine_05, loc=<5.438 5.8653 -6.5241>, rot=<0.9139 0.1791 -0.26 0.2552>}
 dyn_collar_fr_01_L {parent=dyn_collar_fr_MSTR_L, loc=<0.5 0 0>, rot=<0 0 0 1>}
 dyn_collar_fr_01_L1 {parent=dyn_collar_fr_01_L, loc=<4 0 0>, rot=<0 0 0 1>}
 dyn_strap_upper_CST_L {parent=dyn_collar_fr_01_L1, loc=<2.7786 -1.2296 0.9507>, rot=<0.0069 0.0164 0.9197 -0.3923>}
 dyn_strap_upper_01_L {parent=dyn_strap_upper_CST_L, loc=<0.7 0 0>, rot=<-0.1289 0.1185 -0.3309 -0.9273>}
 dyn_strap_upper_02_L {parent=dyn_strap_upper_01_L, loc=<1.2439 0 0>, rot=<0 0 0 1>}
 dyn_strap_lower_CST_L {parent=dyn_collar_fr_01_L1, loc=<0.141 -1.2004 0.4274>, rot=<0.0093 0.0161 0.9362 -0.3509>}
 dyn_strap_lower_01_L {parent=dyn_strap_lower_CST_L, loc=<0.8 0 0>, rot=<-0.0523 0.1299 -0.3698 -0.9185>}
 dyn_strap_lower_02_L {parent=dyn_strap_lower_01_L, loc=<1.2318 0 0>, rot=<0 0 0 1>}
 dyn_collar_fr_MSTR_R {parent=spine_05, loc=<5.4385 5.8653 6.5241>, rot=<-0.2552 0.26 -0.1791 -0.9139>}
 dyn_collar_fr_01_R {parent=dyn_collar_fr_MSTR_R, loc=<0.4993 0.0003 0.0004>, rot=<0 0 0 1>}
 dyn_collar_fr_01_R1 {parent=dyn_collar_fr_01_R, loc=<4.0007 0 0>, rot=<0 0 0 1>}
 dyn_strap_upper_CST_R {parent=dyn_collar_fr_01_R1, loc=<2.7782 1.2294 0.9505>, rot=<0.0069 -0.0164 0.9197 0.3923>}
 dyn_strap_upper_01_R {parent=dyn_strap_upper_CST_R, loc=<0.7002 0.0001 0.0001>, rot=<0.1289 0.1185 0.3309 -0.9273>}
 dyn_strap_upper_02_R {parent=dyn_strap_upper_01_R, loc=<1.2438 0 0>, rot=<0 -1 0 0>}
 dyn_strap_lower_CST_R {parent=dyn_collar_fr_01_R1, loc=<0.1406 1.2003 0.4273>, rot=<0.0093 -0.0161 0.9362 0.3509>}
 dyn_strap_lower_01_R {parent=dyn_strap_lower_CST_R, loc=<0.7997 -0.0001 -0.0001>, rot=<0.0523 0.1299 0.3698 -0.9185>}
 dyn_strap_lower_02_R {parent=dyn_strap_lower_01_R, loc=<1.2319 0.0002 0.0001>, rot=<0 -1 0 0>}
 neck_no_twist {parent=spine_05, loc=<11.3138 0.1529 0>, rot=<0 0 -0.2003 -0.9797>}
 dyn_collar_rear_MSTR {parent=neck_no_twist, loc=<0.9259 -3.736 0.114>, rot=<0.9204 -0.3911 0 0>}
 dyn_collar_rear_01 {parent=dyn_collar_rear_MSTR, loc=<0.5 0 0>, rot=<0 0 0 1>}
 dyn_collar_rear_MSTR_R {parent=neck_no_twist, loc=<0.7743 -0.7657 5.3895>, rot=<-0.2808 -0.8528 0.3697 -0.239>}
 dyn_collar_rear_01_R {parent=dyn_collar_rear_MSTR_R, loc=<-0.4998 -0.0001 -0.0001>, rot=<0 0 0 1>}
 dyn_collar_rear_MSTR_L {parent=neck_no_twist, loc=<0.7745 -0.7657 -5.3895>, rot=<0.8528 -0.2808 -0.239 -0.3697>}
 dyn_collar_rear_01_L {parent=dyn_collar_rear_MSTR_L, loc=<0.5 0 0>, rot=<0 0 0 1>}
 l_collar_dist_check {controlType=EULER_TRANSFORM, parent=spine_05, type=Control, loc=<0 0 0>, rot=<0 0 0 1>}
 r_collar_dist_check {controlType=EULER_TRANSFORM, parent=spine_05, type=Control, loc=<0 0 0>, rot=<0 0 0 1>}
 dyn_coat_fr_MSTR_L {parent=spine_03, loc=<7.2365 8.2808 -9.0966>, rot=<0.0729 -0.9499 -0.3022 0.034>}
 dyn_coat_fr_01_L {parent=dyn_coat_fr_MSTR_L, loc=<1 0 0>, rot=<0 0 0 1>}
 dyn_coat_fr_02_L {parent=dyn_coat_fr_01_L, loc=<7.7003 0 0>, rot=<0 0.0629 0 -0.998>}
 dyn_coat_fr_03_L {parent=dyn_coat_fr_02_L, loc=<8.2749 0 0>, rot=<-0.0018 0.059 0.0311 -0.9978>}
 dyn_coat_fr_04_L {parent=dyn_coat_fr_03_L, loc=<17.9993 0 0.0001>, rot=<0 -0.013 0 -0.9999>}
 dyn_coat_fr_05_L {parent=dyn_coat_fr_04_L, loc=<18 0 0>, rot=<0 -0.0087 0 -1>}
 dyn_coat_fr_06_L {parent=dyn_coat_fr_05_L, loc=<18 0.0001 -0.0001>, rot=<0 0 0 1>}
 dyn_coat_fr_MSTR_R {parent=spine_03, loc=<7.2369 8.2809 9.0966>, rot=<-0.034 0.3022 0.9499 -0.0729>}
 dyn_coat_fr_01_R {parent=dyn_coat_fr_MSTR_R, loc=<1 0 0>, rot=<0 0 0 1>}
 dyn_coat_fr_02_R {parent=dyn_coat_fr_01_R, loc=<7.7003 0 0>, rot=<0 0.0629 0 -0.998>}
 dyn_coat_fr_03_R {parent=dyn_coat_fr_02_R, loc=<8.2749 0 0>, rot=<0.0018 0.059 -0.0311 -0.9978>}
 dyn_coat_fr_04_R {parent=dyn_coat_fr_03_R, loc=<17.9998 0 0>, rot=<0 -0.013 0 -0.9999>}
 dyn_coat_fr_05_R {parent=dyn_coat_fr_04_R, loc=<18.0001 0 0>, rot=<0 -0.0087 0 -1>}
 dyn_coat_fr_06_R {parent=dyn_coat_fr_05_R, loc=<18 0 -0.0001>, rot=<0 0 0 1>}
 dyn_coat_mid_MSTR_L {parent=spine_03, loc=<8.3651 -6.7128 -8.7576>, rot=<0.009 -0.904 0.4263 0.0307>}
 dyn_coat_mid_01_L {parent=dyn_coat_mid_MSTR_L, loc=<1 0 0>, rot=<0 0 0 1>}
 dyn_coat_mid_02_L {parent=dyn_coat_mid_01_L, loc=<15.5081 0 0>, rot=<0 0.2047 0 -0.9788>}
 dyn_coat_mid_03_L {parent=dyn_coat_mid_02_L, loc=<18 0 0>, rot=<0 -0.0729 0 -0.9973>}
 dyn_coat_mid_04_L {parent=dyn_coat_mid_03_L, loc=<18.0001 0 0>, rot=<0 0.0127 0 -0.9999>}
 dyn_coat_mid_05_L {parent=dyn_coat_mid_04_L, loc=<18 0 0>, rot=<0 0 0 -1>}
 dyn_coat_bk_MSTR {parent=spine_03, loc=<8.8655 -8.345 0.0722>, rot=<-0.0046 -0.6985 0.7156 0.0016>}
 dyn_coat_bk_01 {parent=dyn_coat_bk_MSTR, loc=<1 0 0>, rot=<0 0 0 1>}
 dyn_coat_bk_02 {parent=dyn_coat_bk_01, loc=<16.0344 0 0>, rot=<0 0.1868 0 -0.9824>}
 dyn_coat_bk_03 {parent=dyn_coat_bk_02, loc=<18 0 0>, rot=<0 -0.0762 0 -0.9971>}
 dyn_coat_bk_04 {parent=dyn_coat_bk_03, loc=<18 0 0>, rot=<0 0.0103 0 -0.9999>}
 dyn_coat_bk_05 {parent=dyn_coat_bk_04, loc=<18 0 0>, rot=<0 0 0 1>}
 dyn_coat_mid_MSTR_R {parent=spine_03, loc=<8.3655 -6.7128 8.7576>, rot=<-0.0307 -0.4263 0.904 -0.009>}
 dyn_coat_mid_01_R {parent=dyn_coat_mid_MSTR_R, loc=<1 0 0>, rot=<0 0 0 1>}
 dyn_coat_mid_02_R {parent=dyn_coat_mid_01_R, loc=<15.5086 0 0>, rot=<0 0.2047 0 -0.9788>}
 dyn_coat_mid_03_R {parent=dyn_coat_mid_02_R, loc=<18 0 0>, rot=<0 -0.0729 0 -0.9973>}
 dyn_coat_mid_04_R {parent=dyn_coat_mid_03_R, loc=<18 0 0>, rot=<0 0.0127 0 -0.9999>}
 dyn_coat_mid_05_R {parent=dyn_coat_mid_04_R, loc=<18 0 0>, rot=<0 0 0 1>}
 dyn_coat_MSTR_L {parent=spine_03, loc=<7.2371 1.0567 -11.6797>, rot=<0.0578 -0.9942 0.0907 -0.0048>}
 dyn_coat_01_L {parent=dyn_coat_MSTR_L, loc=<1 0 0>, rot=<0 0 0 1>}
 dyn_coat_02_L {parent=dyn_coat_01_L, loc=<15.1203 0 0>, rot=<0 0.1494 0 -0.9888>}
 dyn_coat_03_L {parent=dyn_coat_02_L, loc=<17.9999 0 0.0001>, rot=<0 -0.0598 0 -0.9982>}
 dyn_coat_04_L {parent=dyn_coat_03_L, loc=<18 0 0>, rot=<0 -0.0022 0 -1>}
 dyn_coat_05_L {parent=dyn_coat_04_L, loc=<18 0 0.0001>, rot=<0 0 0 1>}
 dyn_coat_MSTR_R {parent=spine_03, loc=<7.2375 1.0568 11.6797>, rot=<0.0048 -0.0907 0.9942 -0.0578>}
 dyn_coat_01_R {parent=dyn_coat_MSTR_R, loc=<1 0 0>, rot=<0 0 0 1>}
 dyn_coat_02_R {parent=dyn_coat_01_R, loc=<15.121 0 0>, rot=<0 0.1494 0 -0.9888>}
 dyn_coat_03_R {parent=dyn_coat_02_R, loc=<17.9996 0 0.0001>, rot=<0 -0.0598 0 -0.9982>}
 dyn_coat_04_R {parent=dyn_coat_03_R, loc=<18 0 0>, rot=<0 -0.0022 0 -1>}
 dyn_coat_05_R {parent=dyn_coat_04_R, loc=<18 0 0>, rot=<0 0 0 1>}
 thigh_l {parent=pelvis, loc=<-1.0855 -0.2109 -8.7053>, rot=<0.0009 0.0355 0.0155 -0.9993>}
 calf_l {parent=thigh_l, loc=<-42.5284 -0.0019 -0.7297>, rot=<0.0006 -0.0161 -0.0391 -0.9991>}
 foot_l {parent=calf_l, loc=<-46.7401 -2.5014 -0.592>, rot=<0.0427 -0.0238 0.0353 -0.9982>}
 ball_l {parent=foot_l, loc=<-3.5658 10.8066 -0.1544>, rot=<0.0001 -0.0001 -0.7063 -0.7079>}
 calf_twist_02_l {parent=calf_l, loc=<-19.1546 -1.7943 -0.2638>, rot=<0.0237 0.0095 -0.0035 -0.9997>}
 calf_twist_01_l {parent=calf_l, loc=<-35.8483 -1.9197 -0.5794>, rot=<0.0237 0.0095 -0.0035 -0.9997>}
 deform_knee_in_base_l {parent=calf_l, loc=<0.8031 -5.0231 0.2456>, rot=<0 0 0 -1>}
 deform_knee_in_l {parent=deform_knee_in_base_l, loc=<0 0 0>, rot=<0 0 0 1>}
 deform_knee_out_base_l {parent=calf_l, loc=<0 4 0>, rot=<0 0 0 -1>}
 deform_knee_out_l {parent=deform_knee_out_base_l, loc=<0 0 0>, rot=<0 0 0 1>}
 thigh_twist_01_l {parent=thigh_l, loc=<-27.952 0 0>, rot=<0.0013 0 0 -1>}
 deform_groin_base_l {parent=thigh_l, loc=<0.4801 5.6776 -0.2486>, rot=<0 0 0 -1>}
 deform_groin_l {parent=deform_groin_base_l, loc=<0 0 0>, rot=<0 0 0 1>}
 thigh_r {parent=pelvis, loc=<-1.0855 -0.2109 8.7053>, rot=<0.0355 -0.0009 0.9993 0.0155>}
 calf_r {parent=thigh_r, loc=<42.5284 0.0019 0.7297>, rot=<0.0006 -0.0161 -0.0391 -0.9991>}
 foot_r {parent=calf_r, loc=<46.7401 2.5014 0.592>, rot=<0.0427 -0.0238 0.0353 -0.9982>}
 ball_r {parent=foot_r, loc=<3.5658 -10.8066 0.1544>, rot=<0.0001 -0.0001 -0.7063 -0.7079>}
 calf_twist_02_r {parent=calf_r, loc=<19.1546 1.7943 0.2638>, rot=<0.0237 0.0095 -0.0035 -0.9997>}
 calf_twist_01_r {parent=calf_r, loc=<35.8483 1.9197 0.5794>, rot=<0.0237 0.0095 -0.0035 -0.9997>}
 deform_knee_in_base_r {parent=calf_r, loc=<-0.8031 5.0231 -0.2457>, rot=<0 0 0 -1>}
 deform_knee_in_r {parent=deform_knee_in_base_r, loc=<0 0 0>, rot=<0 0 0 1>}
 deform_knee_out_base_r {parent=calf_r, loc=<0 -4 0>, rot=<0 0 0 -1>}
 deform_knee_out_r {parent=deform_knee_out_base_r, loc=<0 0 0>, rot=<0 0 0 1>}
 thigh_twist_01_r {parent=thigh_r, loc=<27.952 0 0>, rot=<0.0013 0 0 -1>}
 deform_groin_base_r {parent=thigh_r, loc=<-0.4801 -5.6776 0.2486>, rot=<0 0 0 -1>}
 deform_groin_r {parent=deform_groin_base_r, loc=<0 0 0>, rot=<0 0 0 1>}
]
variables [
 IsCloth {type=bool, default=False}
 IsWearingCape {type=bool, default=False}
]
