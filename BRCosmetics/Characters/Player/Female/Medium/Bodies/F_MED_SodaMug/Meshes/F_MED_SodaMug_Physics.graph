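# Physics asset for the F_MED_SodaMug female player body, set up for a dynamic long coat.
# Seven coat capsule chains (front, side and mid pairs, one back) hang from MSTR bones with limited swing and twist.
# Fully free constraints cross-link neighboring chains at several segments.
# Spine, pelvis and leg capsules form the body shapes, and root spheres include a huge one below it, resembling a floor.
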PhysicsAsset /BRCosmetics/Characters/Player/Female/Medium/Bodies/F_MED_SodaMug/Meshes/F_MED_SodaMug_Physics
bodies {
 calf_l [{shapeType=Capsule, length=39.2812, radius=7, shapeName=calf_r_capsule, center=<-23.3608 -2.5 0>, rotation=<rot -90 0 0>}]
 calf_r [{shapeType=Capsule, length=32.375, radius=9.4062, shapeName=calf_r_capsule, center=<23.4118 3.9866 0>, rotation=<rot -90 180 180>}]
 dyn_coat_01_L [{shapeType=Capsule, length=11.25, radius=4.5625, shapeName=dyn_coat_mid_01_L_capsule, center=<13 0 5>, rotation=<rot -82.9332 0.002 0>}]
 dyn_coat_01_R [{
   shapeType=Capsule
   length=11.25
   radius=4.5625
   shapeName=dyn_coat_mid_01_R_capsule
   center=<13.0944 0 4.2387>
   rotation=<rot -82.9332 -0.002 0>
  }]
 dyn_coat_02_L [{shapeType=Capsule, length=18.5, radius=5.5, shapeName=dyn_coat_fr_03_L_capsule, center=<13 0 5>, rotation=<rot -90 179.9965 -179.9943>}]
 dyn_coat_02_R [{shapeType=Capsule, length=18.5, radius=5.5, shapeName=dyn_coat_fr_03_L_capsule, center=<13 0 5>, rotation=<rot -90 179.9965 -179.9943>}]
 dyn_coat_03_L [{shapeType=Capsule, length=14, radius=6.5, shapeName=dyn_coat_mid_03_L_capsule, center=<13 0 5>, rotation=<rot -90 179.9956 -179.9938>}]
 dyn_coat_03_R [{shapeType=Capsule, length=16.4812, radius=7.66, shapeName=dyn_coat_mid_03_R_capsule, center=<13 0 5>, rotation=<rot -90 -0.021 0.0191>}]
 dyn_coat_04_L [{shapeType=Capsule, length=14, radius=6.5, shapeName=dyn_coat_fr_05_R_capsule, center=<13 0 5>, rotation=<rot -90 -179.9991 179.9969>}]
 dyn_coat_04_R [{shapeType=Capsule, length=14, radius=6.5, shapeName=dyn_coat_fr_05_R_capsule, center=<13 0 5>, rotation=<rot -90 -179.9991 179.9969>}]
 dyn_coat_05_L [{shapeType=Capsule, length=14, radius=8.5, shapeName=dyn_coat_fr_06_R_capsule, center=<13 0 5>, rotation=<rot -90 -17.4099 -72.5929>}]
 dyn_coat_05_R [{shapeType=Capsule, length=14, radius=8.5, shapeName=dyn_coat_fr_06_R_capsule, center=<13 0 5>, rotation=<rot -90 -17.4099 -72.5929>}]
 dyn_coat_MSTR_L [{shapeType=Capsule, length=0.505, radius=0.505, shapeName=dyn_coat_MSTR_L_capsule, rotation=<rot 0 -90 0>}]
 dyn_coat_MSTR_R [{shapeType=Capsule, length=0.505, radius=0.505, shapeName=dyn_coat_MSTR_R_capsule, rotation=<rot 0 -90 0>}]
 dyn_coat_bk_01 [{shapeType=Capsule, length=10.0312, radius=4.4688, shapeName=dyn_coat_bk_01_capsule, center=<13 0 5>, rotation=<rot -76.9676 -0.0001 0>}]
 dyn_coat_bk_02 [{shapeType=Capsule, length=17.1562, radius=6.8438, shapeName=dyn_coat_fr_03_L_capsule, center=<13 0 5>, rotation=<rot -82.8531 -179.9978 180>}]
 dyn_coat_bk_03 [{shapeType=Capsule, length=14, radius=6.5, shapeName=dyn_coat_bk_03_capsule, center=<13 0 5>, rotation=<rot -90 -179.9605 179.9604>}]
 dyn_coat_bk_04 [{shapeType=Capsule, length=14, radius=6.5, shapeName=dyn_coat_fr_05_R_capsule, center=<13 0 5>, rotation=<rot 0 0 90>}]
 dyn_coat_bk_05 [{shapeType=Capsule, length=23.9062, radius=8.5, shapeName=dyn_coat_fr_06_R_capsule, center=<13 0 5>, rotation=<rot -90 -17.4099 -72.5929>}]
 dyn_coat_bk_MSTR [{shapeType=Capsule, length=0, radius=0.5, shapeName=dyn_coat_bk_MSTR_capsule}]
 dyn_coat_fr_01_L [{shapeType=Capsule, length=13.6875, radius=3.5, shapeName=dyn_coat_fr_01_L_capsule, center=<13 0 5>, rotation=<rot -83.8458 0.0022 0>}]
 dyn_coat_fr_01_R [{shapeType=Capsule, length=13.3438, radius=3.8438, shapeName=dyn_coat_fr_01_R_capsule, center=<13 0 5>, rotation=<rot -81.3148 -0.0023 0>}]
 dyn_coat_fr_02_L [{shapeType=Capsule, length=15.9375, radius=4.5625, shapeName=dyn_coat_fr_01_L_capsule, center=<13 0 5>, rotation=<rot -82.5295 0.0023 0>}]
 dyn_coat_fr_02_R [{
   shapeType=Capsule
   length=15.9062
   radius=4.5938
   shapeName=dyn_coat_fr_01_R_capsule
   center=<13.0163 0 4.8754>
   rotation=<rot -82.5295 -0.0023 0>
  }]
 dyn_coat_fr_03_L [{shapeType=Capsule, length=18.5, radius=5.5, shapeName=dyn_coat_fr_03_L_capsule, center=<13 0 5>, rotation=<rot -90 179.9965 -179.9943>}]
 dyn_coat_fr_03_R [{
   shapeType=Capsule
   length=17.5683
   radius=6.4317
   shapeName=dyn_coat_fr_03_R_capsule
   center=<13 0 4.6947>
   rotation=<rot -90 -179.9991 179.997>
  }]
 dyn_coat_fr_04_L [{shapeType=Capsule, length=13.5, radius=6, shapeName=dyn_coat_fr_04_L_capsule, center=<13 0 5>, rotation=<rot -90 172.4484 -172.4462>}]
 dyn_coat_fr_04_R [{shapeType=Capsule, length=15.4975, radius=7.6843, shapeName=dyn_coat_fr_04_R_capsule, center=<13 0 5>, rotation=<rot -90 0 -0.0022>}]
 dyn_coat_fr_05_L [{shapeType=Capsule, length=14, radius=6.5, shapeName=dyn_coat_fr_05_L_capsule, center=<13 0 5>, rotation=<rot -90 179.8948 -179.8926>}]
 dyn_coat_fr_05_R [{shapeType=Capsule, length=14, radius=6.5, shapeName=dyn_coat_fr_05_R_capsule, center=<13 0 5>, rotation=<rot -90 -179.9991 179.9969>}]
 dyn_coat_fr_06_L [{shapeType=Capsule, length=15.5, radius=8.5, shapeName=dyn_coat_fr_06_L_capsule, center=<13 0 5>, rotation=<rot -90 17.4099 -17.409>}]
 dyn_coat_fr_06_R [{shapeType=Capsule, length=14, radius=8.5, shapeName=dyn_coat_fr_06_R_capsule, center=<13 0 5>, rotation=<rot -90 -17.4099 17.4068>}]
 dyn_coat_fr_MSTR_L [{shapeType=Capsule, length=0, radius=0.5, shapeName=dyn_coat_fr_MSTR_L_capsule}]
 dyn_coat_fr_MSTR_R [{shapeType=Capsule, length=0, radius=0.5, shapeName=dyn_coat_fr_MSTR_R_capsule}]
 dyn_coat_mid_01_L [{
   shapeType=Capsule
   length=10.0312
   radius=4.4688
   shapeName=dyn_coat_mid_01_L_capsule
   center=<13 0 5>
   rotation=<rot -76.9675 0.0019 0.0001>
  }]
 dyn_coat_mid_01_R [{shapeType=Capsule, length=10.0312, radius=4.4688, shapeName=dyn_coat_mid_01_R_capsule, center=<13 0 5>, rotation=<rot -76.9674 -0.0019 0>}]
 dyn_coat_mid_02_L [{
   shapeType=Capsule
   length=17.1562
   radius=6.8438
   shapeName=dyn_coat_fr_03_L_capsule
   center=<12.8989 0 4.4766>
   rotation=<rot -79.0643 -179.9979 -179.9998>
  }]
 dyn_coat_mid_02_R [{
   shapeType=Capsule
   length=17.1562
   radius=6.8438
   shapeName=dyn_coat_fr_03_L_capsule
   center=<12.8989 0 4.4766>
   rotation=<rot -79.0643 -179.9977 179.9998>
  }]
 dyn_coat_mid_03_L [{shapeType=Capsule, length=14, radius=6.5, shapeName=dyn_coat_mid_03_L_capsule, center=<13 0 5>, rotation=<rot -90 179.9956 -179.9938>}]
 dyn_coat_mid_03_R [{shapeType=Capsule, length=14, radius=6.5, shapeName=dyn_coat_mid_03_R_capsule, center=<13 0 5>, rotation=<rot -90 -0.021 0.0191>}]
 dyn_coat_mid_04_L [{shapeType=Capsule, length=14, radius=6.5, shapeName=dyn_coat_fr_05_R_capsule, center=<13 0 5>, rotation=<rot 0 0 90>}]
 dyn_coat_mid_04_R [{shapeType=Capsule, length=14, radius=6.5, shapeName=dyn_coat_fr_05_R_capsule, center=<13 0 5>, rotation=<rot 0 0 90>}]
 dyn_coat_mid_05_L [{shapeType=Capsule, length=26.8125, radius=8.5, shapeName=dyn_coat_fr_06_R_capsule, center=<13 0 5>, rotation=<rot -90 -17.4099 -72.5929>}]
 dyn_coat_mid_05_R [{shapeType=Capsule, length=26.8125, radius=8.5, shapeName=dyn_coat_fr_06_R_capsule, center=<13 0 5>, rotation=<rot -90 -17.4099 -72.5929>}]
 dyn_coat_mid_MSTR_L [{shapeType=Capsule, length=0, radius=0.5, shapeName=dyn_coat_mid_MSTR_L_capsule}]
 dyn_coat_mid_MSTR_R [{shapeType=Capsule, length=0, radius=0.5, shapeName=dyn_coat_mid_MSTR_R_capsule}]
 foot_l [{shapeType=Capsule, length=17.8438, radius=9.5, shapeName=foot_r_capsule, center=<-3 7 0>, rotation=<rot 0 0 90>}]
 foot_r [{shapeType=Capsule, length=17.8438, radius=9.5, shapeName=foot_r_capsule, center=<3 -7 0>, rotation=<rot 0 0 -90>}]
 pelvis [{shapeType=Capsule, length=10.6562, radius=9, shapeName=pelvis_capsule}]
 root [
  {shapeType=Sphere, radius=300, shapeName=root_sphere, center=<0 0 -289.119>}
  {shapeType=Sphere, radius=42, shapeName=root_sphere, center=<0 5.5883 -12.006>}
 ]
 spine_01 [{shapeType=Capsule, length=2.375, radius=10.6562, shapeName=spine_01_capsule, center=<3.5975 0.2921 0>}]
 spine_02 [{shapeType=Capsule, length=8.625, radius=6.4688, shapeName=spine_02_capsule, center=<4.8113 2.6414 0>}]
 spine_03 [{shapeType=Capsule, length=7.125, radius=7.5, shapeName=spine_03_capsule, center=<4.4602 2.6414 0>}]
 thigh_l [{shapeType=Capsule, length=35.75, radius=8.5, shapeName=thigh_r_capsule, center=<-21 0 0>, rotation=<rot -88.2973 -0.0005 0.0003>}]
 thigh_r [{
   shapeType=Capsule
   length=35.0924
   radius=9.1576
   shapeName=thigh_r_capsule
   center=<20.977 0 0.773>
   rotation=<rot -85.5368 -0.0004 -179.9998>
  }]
}
constraints [
 {
  bone1Name=spine_01
  bone2Name=pelvis
  swing1LimitDegrees=45
  swing1Motion=Free
  swing2LimitDegrees=45
  swing2Motion=Free
  twistLimitDegrees=45
  twistMotion=Free
 }
 {
  bone1Name=spine_02
  bone2Name=spine_01
  swing1LimitDegrees=45
  swing1Motion=Free
  swing2LimitDegrees=45
  swing2Motion=Free
  twistLimitDegrees=45
  twistMotion=Free
 }
 {
  bone1Name=spine_03
  bone2Name=spine_02
  swing1LimitDegrees=45
  swing1Motion=Free
  swing2LimitDegrees=45
  swing2Motion=Free
  twistLimitDegrees=45
  twistMotion=Free
 }
 {
  bone1Name=thigh_l
  bone2Name=pelvis
  swing1LimitDegrees=45
  swing1Motion=Free
  swing2LimitDegrees=45
  swing2Motion=Free
  twistLimitDegrees=45
  twistMotion=Free
 }
 {
  bone1Name=calf_l
  bone2Name=thigh_l
  swing1LimitDegrees=45
  swing1Motion=Free
  swing2LimitDegrees=45
  swing2Motion=Free
  twistLimitDegrees=45
  twistMotion=Free
 }
 {
  bone1Name=foot_l
  bone2Name=calf_l
  swing1LimitDegrees=45
  swing1Motion=Free
  swing2LimitDegrees=45
  swing2Motion=Free
  twistLimitDegrees=45
  twistMotion=Free
 }
 {
  bone1Name=calf_r
  bone2Name=thigh_r
  swing1LimitDegrees=45
  swing1Motion=Free
  swing2LimitDegrees=45
  swing2Motion=Free
  twistLimitDegrees=45
  twistMotion=Free
 }
 {
  bone1Name=foot_r
  bone2Name=calf_r
  swing1LimitDegrees=45
  swing1Motion=Free
  swing2LimitDegrees=45
  swing2Motion=Free
  twistLimitDegrees=45
  twistMotion=Free
 }
 {bone1Name=dyn_coat_bk_01, bone2Name=dyn_coat_bk_MSTR, swing1LimitDegrees=8, swing2LimitDegrees=18, twistLimitDegrees=8}
 {bone1Name=dyn_coat_bk_03, bone2Name=dyn_coat_bk_02, swing1LimitDegrees=8, swing2LimitDegrees=18, twistLimitDegrees=8}
 {bone1Name=dyn_coat_bk_04, bone2Name=dyn_coat_bk_03, swing1LimitDegrees=12, swing2LimitDegrees=20, twistLimitDegrees=8}
 {bone1Name=dyn_coat_mid_01_L, bone2Name=dyn_coat_mid_MSTR_L, swing1LimitDegrees=8, swing2LimitDegrees=18, twistLimitDegrees=8}
 {bone1Name=dyn_coat_mid_03_L, bone2Name=dyn_coat_mid_02_L, swing1LimitDegrees=8, swing2LimitDegrees=18, twistLimitDegrees=8}
 {bone1Name=dyn_coat_mid_04_L, bone2Name=dyn_coat_mid_03_L, swing1LimitDegrees=12, swing2LimitDegrees=20, twistLimitDegrees=8}
 {bone1Name=dyn_coat_mid_01_R, bone2Name=dyn_coat_mid_MSTR_R, swing1LimitDegrees=8, swing2LimitDegrees=18, twistLimitDegrees=8}
 {bone1Name=dyn_coat_mid_03_R, bone2Name=dyn_coat_mid_02_R, swing1LimitDegrees=8, swing2LimitDegrees=18, twistLimitDegrees=8}
 {bone1Name=dyn_coat_mid_04_R, bone2Name=dyn_coat_mid_03_R, swing1LimitDegrees=12, swing2LimitDegrees=32, twistLimitDegrees=8}
 {bone1Name=dyn_coat_fr_01_L, bone2Name=dyn_coat_fr_MSTR_L, swing1LimitDegrees=8, swing2LimitDegrees=18, twistLimitDegrees=8}
 {bone1Name=dyn_coat_fr_02_L, bone2Name=dyn_coat_fr_01_L, swing1LimitDegrees=8, swing2LimitDegrees=18, twistLimitDegrees=8}
 {bone1Name=dyn_coat_fr_03_L, bone2Name=dyn_coat_fr_02_L, swing1LimitDegrees=8, swing2LimitDegrees=18, twistLimitDegrees=8}
 {bone1Name=dyn_coat_fr_04_L, bone2Name=dyn_coat_fr_03_L, swing1LimitDegrees=12, swing2LimitDegrees=32, twistLimitDegrees=8}
 {bone1Name=dyn_coat_fr_05_L, bone2Name=dyn_coat_fr_04_L, swing1LimitDegrees=12, swing2LimitDegrees=32, twistLimitDegrees=8}
 {bone1Name=dyn_coat_fr_06_L, bone2Name=dyn_coat_fr_05_L, swing1LimitDegrees=12, swing2LimitDegrees=32, twistLimitDegrees=8}
 {bone1Name=dyn_coat_fr_01_R, bone2Name=dyn_coat_fr_MSTR_R, swing1LimitDegrees=8, swing2LimitDegrees=18, twistLimitDegrees=8}
 {bone1Name=dyn_coat_fr_02_R, bone2Name=dyn_coat_fr_01_R, swing1LimitDegrees=8, swing2LimitDegrees=18, twistLimitDegrees=8}
 {bone1Name=dyn_coat_fr_03_R, bone2Name=dyn_coat_fr_02_R, swing1LimitDegrees=8, swing2LimitDegrees=18, twistLimitDegrees=8}
 {bone1Name=dyn_coat_fr_04_R, bone2Name=dyn_coat_fr_03_R, swing1LimitDegrees=12, swing2LimitDegrees=32, twistLimitDegrees=8}
 {bone1Name=dyn_coat_fr_05_R, bone2Name=dyn_coat_fr_04_R, swing1LimitDegrees=12, swing2LimitDegrees=32, twistLimitDegrees=8}
 {bone1Name=dyn_coat_fr_06_R, bone2Name=dyn_coat_fr_05_R, swing1LimitDegrees=12, swing2LimitDegrees=32, twistLimitDegrees=8}
 {bone1Name=dyn_coat_01_L, bone2Name=dyn_coat_MSTR_L, swing1LimitDegrees=8, swing2LimitDegrees=18, twistLimitDegrees=8}
 {bone1Name=dyn_coat_03_L, bone2Name=dyn_coat_02_L, swing1LimitDegrees=8, swing2LimitDegrees=18, twistLimitDegrees=8}
 {bone1Name=dyn_coat_04_L, bone2Name=dyn_coat_03_L, swing1LimitDegrees=12, swing2LimitDegrees=32, twistLimitDegrees=8}
 {bone1Name=dyn_coat_01_R, bone2Name=dyn_coat_MSTR_R, swing1LimitDegrees=8, swing2LimitDegrees=18, twistLimitDegrees=8}
 {bone1Name=dyn_coat_03_R, bone2Name=dyn_coat_02_R, swing1LimitDegrees=8, swing2LimitDegrees=18, twistLimitDegrees=8}
 {bone1Name=dyn_coat_04_R, bone2Name=dyn_coat_03_R, swing1LimitDegrees=12, swing2LimitDegrees=32, twistLimitDegrees=8}
 {bone1Name=dyn_coat_05_L, bone2Name=dyn_coat_04_L, swing1LimitDegrees=12, swing2LimitDegrees=32, twistLimitDegrees=8}
 {bone1Name=dyn_coat_05_R, bone2Name=dyn_coat_04_R, swing1LimitDegrees=12, swing2LimitDegrees=32, twistLimitDegrees=8}
 {bone1Name=dyn_coat_bk_05, bone2Name=dyn_coat_bk_04, swing1LimitDegrees=12, swing2LimitDegrees=20, twistLimitDegrees=8}
 {bone1Name=dyn_coat_mid_05_L, bone2Name=dyn_coat_mid_04_L, swing1LimitDegrees=12, swing2LimitDegrees=32, twistLimitDegrees=8}
 {bone1Name=dyn_coat_mid_05_R, bone2Name=dyn_coat_mid_04_R, swing1LimitDegrees=12, swing2LimitDegrees=32, twistLimitDegrees=8}
 {
  bone1Name=dyn_coat_fr_04_R
  bone2Name=dyn_coat_03_R
  swing1LimitDegrees=45
  swing1Motion=Free
  swing2LimitDegrees=45
  swing2Motion=Free
  twistLimitDegrees=45
  twistMotion=Free
 }
 {
  bone1Name=dyn_coat_fr_05_R
  bone2Name=dyn_coat_04_R
  swing1LimitDegrees=45
  swing1Motion=Free
  swing2LimitDegrees=45
  swing2Motion=Free
  twistLimitDegrees=45
  twistMotion=Free
 }
 {
  bone1Name=dyn_coat_04_R
  bone2Name=dyn_coat_mid_04_R
  swing1LimitDegrees=45
  swing1Motion=Free
  swing2LimitDegrees=45
  swing2Motion=Free
  twistLimitDegrees=45
  twistMotion=Free
 }
 {
  bone1Name=dyn_coat_fr_04_L
  bone2Name=dyn_coat_03_L
  swing1LimitDegrees=45
  swing1Motion=Free
  swing2LimitDegrees=45
  swing2Motion=Free
  twistLimitDegrees=45
  twistMotion=Free
 }
 {
  bone1Name=dyn_coat_fr_05_L
  bone2Name=dyn_coat_04_L
  swing1LimitDegrees=45
  swing1Motion=Free
  swing2LimitDegrees=45
  swing2Motion=Free
  twistLimitDegrees=45
  twistMotion=Free
 }
 {
  bone1Name=dyn_coat_04_L
  bone2Name=dyn_coat_mid_04_L
  swing1LimitDegrees=45
  swing1Motion=Free
  swing2LimitDegrees=45
  swing2Motion=Free
  twistLimitDegrees=45
  twistMotion=Free
 }
 {
  bone1Name=dyn_coat_mid_03_L
  bone2Name=dyn_coat_bk_03
  swing1LimitDegrees=45
  swing1Motion=Free
  swing2LimitDegrees=45
  swing2Motion=Free
  twistLimitDegrees=45
  twistMotion=Free
 }
 {
  bone1Name=dyn_coat_mid_03_R
  bone2Name=dyn_coat_bk_03
  swing1LimitDegrees=45
  swing1Motion=Free
  swing2LimitDegrees=45
  swing2Motion=Free
  twistLimitDegrees=45
  twistMotion=Free
 }
 {bone1Name=dyn_coat_bk_02, bone2Name=dyn_coat_bk_01, swing1LimitDegrees=8, swing2LimitDegrees=18, twistLimitDegrees=8}
 {bone1Name=dyn_coat_02_R, bone2Name=dyn_coat_01_R, swing1LimitDegrees=8, swing2LimitDegrees=18, twistLimitDegrees=8}
 {bone1Name=dyn_coat_02_L, bone2Name=dyn_coat_01_L, swing1LimitDegrees=8, swing2LimitDegrees=18, twistLimitDegrees=8}
 {bone1Name=dyn_coat_mid_02_L, bone2Name=dyn_coat_mid_01_L, swing1LimitDegrees=8, swing2LimitDegrees=18, twistLimitDegrees=8}
 {bone1Name=dyn_coat_mid_02_R, bone2Name=dyn_coat_mid_01_R, swing1LimitDegrees=8, swing2LimitDegrees=18, twistLimitDegrees=8}
 {
  bone1Name=dyn_coat_02_L
  bone2Name=dyn_coat_mid_02_L
  swing1LimitDegrees=45
  swing1Motion=Free
  swing2LimitDegrees=45
  swing2Motion=Free
  twistLimitDegrees=45
  twistMotion=Free
 }
 {
  bone1Name=dyn_coat_mid_02_L
  bone2Name=dyn_coat_bk_02
  swing1LimitDegrees=45
  swing1Motion=Free
  swing2LimitDegrees=45
  swing2Motion=Free
  twistLimitDegrees=45
  twistMotion=Free
 }
 {
  bone1Name=dyn_coat_02_R
  bone2Name=dyn_coat_mid_02_R
  swing1LimitDegrees=45
  swing1Motion=Free
  swing2LimitDegrees=45
  swing2Motion=Free
  twistLimitDegrees=45
  twistMotion=Free
 }
 {
  bone1Name=dyn_coat_mid_02_R
  bone2Name=dyn_coat_bk_02
  swing1LimitDegrees=45
  swing1Motion=Free
  swing2LimitDegrees=45
  swing2Motion=Free
  twistLimitDegrees=45
  twistMotion=Free
 }
 {
  bone1Name=thigh_r
  bone2Name=pelvis
  swing1LimitDegrees=45
  swing1Motion=Free
  swing2LimitDegrees=45
  swing2Motion=Free
  twistLimitDegrees=45
  twistMotion=Free
 }
 {
  bone1Name=dyn_coat_mid_05_L
  bone2Name=dyn_coat_bk_05
  swing1LimitDegrees=45
  swing1Motion=Free
  swing2LimitDegrees=45
  swing2Motion=Free
  twistLimitDegrees=45
  twistMotion=Free
 }
 {
  bone1Name=dyn_coat_mid_05_R
  bone2Name=dyn_coat_bk_05
  swing1LimitDegrees=45
  swing1Motion=Free
  swing2LimitDegrees=45
  swing2Motion=Free
  twistLimitDegrees=45
  twistMotion=Free
 }
 {
  bone1Name=dyn_coat_05_R
  bone2Name=dyn_coat_mid_05_R
  swing1LimitDegrees=45
  swing1Motion=Free
  swing2LimitDegrees=45
  swing2Motion=Free
  twistLimitDegrees=45
  twistMotion=Free
 }
 {
  bone1Name=dyn_coat_fr_06_R
  bone2Name=dyn_coat_05_R
  swing1LimitDegrees=45
  swing1Motion=Free
  swing2LimitDegrees=45
  swing2Motion=Free
  twistLimitDegrees=45
  twistMotion=Free
 }
 {
  bone1Name=dyn_coat_mid_05_L
  bone2Name=dyn_coat_bk_05
  swing1LimitDegrees=45
  swing1Motion=Free
  swing2LimitDegrees=45
  swing2Motion=Free
  twistLimitDegrees=45
  twistMotion=Free
 }
 {
  bone1Name=dyn_coat_fr_06_L
  bone2Name=dyn_coat_mid_05_L
  swing1LimitDegrees=45
  swing1Motion=Free
  swing2LimitDegrees=45
  swing2Motion=Free
  twistLimitDegrees=45
  twistMotion=Free
 }
]
